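# Niagara idle effect for the PunkDevilVibe pickaxe skin, a single sprite emitter named Stars that spawns a slow trickle of short-lived sparkles.
# Sprites spawn in a flattened sphere offset to either side along X and drift with a small added velocity; gravity is disabled.
# A stepped color curve sampled randomly makes each either HDR pink or cyan, while size and alpha curves make it pulse and fade out.
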NiagaraSystem /BRCosmetics/Weapons/FORT_Melee/Pickaxe_Punk_Devil_Summer_Female/Skin/Pickaxe_PunkDevilVibe/FX/NS_Pickaxe_PunkDevilVibe_Idle
dependencies {
 usedDataInterfaces [&NiagaraDataInterfaceColorCurve, &NiagaraDataInterfaceArrayFloat3, &NiagaraDataInterfaceCurve]
 usedDynamicInputs []
 usedModules [&SystemState]
}
emitters {Stars {
  emitterUpdateScript [
   &EmitterState()
   &SpawnRate(SpawnRate=5, Use Spawn Probability=false)
  ]
  particleSpawnScript [
   &InitializeParticle(
    Color=&ColorFromCurve(
     ColorCurve=DataInterface() {propertyValues {
       AlphaCurve {keys [<at 0 1 RCIM_Constant>]}
       BlueCurve {keys [<at 0 0.5668 RCIM_Constant>,<at 0.4077 2 RCIM_Constant>]}
       GreenCurve {keys [<at 0 0.1562 RCIM_Constant>,<at 0.4077 1.9758 RCIM_Constant>]}
       RedCurve {keys [<at 0 2 RCIM_Constant>,<at 0.4077 0.1771 RCIM_Constant>]}
      }}
     CurveIndex=&RandomRangeFloat(Minimum=0, Maximum=1)
    )
    Lifetime Max=0.7
    Lifetime Min=0.2
    Lifetime Mode=::NewEnumerator1
    Sprite Size Mode=::NewEnumerator1
    Uniform Sprite Size Max=25
    Uniform Sprite Size Min=10
    UsePositionOffset=false
    Write Material Random=true
   )
   &ShapeLocation(
    MaskSpawnGroup=false
    Non Uniform Scale=<0.1 1 0.25>
    Offset=&SelectVectorFromArray(
     Array Sampling Mode=::NewEnumerator0
     Direct Array Index=?
     Vector Selection Array=DataInterface() {propertyValues {FloatData [<3 0 0>, <-3 0 0>]}}
     Random Seed=0
    )
    Offset Mode=::NewEnumerator0
    Override Seed=false
    Sphere Radius=20
    UseOwnerScale=true
   )
   &AddVelocity(Module.Use Falloff Distance=false, Override Seed=false, Velocity Mode=::NewEnumerator1, Velocity Speed=20)
  ]
  particleUpdateScript [
   &ParticleState()
   &ScaleSpriteSize(Scale Factor=&Vector2DFromFloat(Value=&FloatFromCurve(FloatCurve=DataInterface() {propertyValues {Curve {keys [
          {arriveTangent=7.1136, interpMode=RCIM_Cubic, leaveTangent=7.1136, tangentMode=RCTM_User, time=0, value=0}
          {arriveTangent=1.125, interpMode=RCIM_Cubic, leaveTangent=1.125, time=0.2, value=0.85}
          <at 0.4 0.45>
          {arriveTangent=-0.0952, interpMode=RCIM_Cubic, leaveTangent=-0.0952, tangentMode=RCTM_User, time=0.7849, value=0.7663}
          {interpMode=RCIM_Cubic, leaveTangent=-1.238, time=1, value=0}
         ]}}})))
   &GravityForce(Gravity=<0 0 -50>) {enabled=false}
   &SolveForcesAndVelocity(Clamp Velocity=false, Limit Acceleration=false)
   &ScaleColor(
    Scale Alpha=&FloatFromCurve(FloatCurve=DataInterface() {propertyValues {Curve {keys [
         {interpMode=RCIM_Cubic, tangentMode=RCTM_User, time=0, value=1}
         {arriveTangent=-2.6953, interpMode=RCIM_Cubic, leaveTangent=-2.6953, tangentMode=RCTM_User, time=1, value=0}
        ]}}})
    ScaleA=true
    ScaleRGB=false
   )
  ]
  rendererClasses [&NiagaraSpriteRendererProperties]
  renderers [{rendererClass=&NiagaraSpriteRendererProperties}]
 }}
summary {
 systemName=NS_Pickaxe_PunkDevilVibe_Idle
 emitters [{emitterName=Stars, rendererClasses [&NiagaraSpriteRendererProperties]}]
 userVariables []
}
